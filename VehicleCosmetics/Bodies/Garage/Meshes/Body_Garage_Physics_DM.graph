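# Physics asset for the Garage vehicle body mesh in the VehicleCosmetics content.
# A single body on the root bone uses one box shape, offset forward and upward to approximate the vehicle's hull.
# It has no constraints, so the body acts as one rigid collision volume.
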PhysicsAsset /VehicleCosmetics/Bodies/Garage/Meshes/Body_Garage_Physics_DM
bodies {root [{shapeType=Box, extentX=444, extentY=216, extentZ=116, shapeName=root_box, center=<35.4637 0 98.5>}]}
constraints []
